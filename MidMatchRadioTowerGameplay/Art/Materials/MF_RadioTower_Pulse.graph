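# Material function that outputs an animated pulse mask, by its naming for the red lights on the mid-match radio tower.
# Red vertex color, halved and inverted, offsets a looping, repeating Time value to form the UVs for a clamped grayscale pulse texture.
# Vertex paint thus sets each point's phase in the pulse without relying on mesh UVs.
# The texture's red channel is the output; trailing scale and bias nodes are unconnected.
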MaterialFunction /MidMatchRadioTowerGameplay/Art/Materials/MF_RadioTower_Pulse
graph {nodes [
  n1 = Multiply(A=VertexColor().R) {constB=0.5}
  n2 = Time() {bOverride_Period=true, period=1.5}
  n3 = TextureSample(UVs=Add(A=OneMinus(None=@n1), B=Add(A=@n2) {constB=-1})) {
   samplerSource=SSM_Clamp_WorldGroupSettings
   samplerType=SAMPLERTYPE_LinearGrayscale
   texture=/MidMatchRadioTowerGameplay/Art/Textures/T_Gameplay_RadioTower_RedLights_Pulse
  }
  Add(A=Multiply(A=@n3.R) {constB=0.8}) {constB=0.6}
  FunctionOutput(None=@n3.R)
  Multiply(A=@n2) {constB=1.5}
 ]}
