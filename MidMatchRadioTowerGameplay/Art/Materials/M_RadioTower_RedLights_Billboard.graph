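# Translucent unlit material for the radio tower's light flares, drawn as camera-facing billboards on Nanite or instanced meshes.
# Vertex offset builds view-aligned quads that grow with camera distance, pulse and are pulled toward the camera.
# Position-hashed randomness and noise vary each flare, layered with opacity, rotating chroma and scrolling spark textures, time-of-day opacity and depth fade.
# Custom primitive data drives active state, activation lights, completion and red-versus-green colors, alongside vertex-color masks.
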Material /MidMatchRadioTowerGameplay/Art/Materials/M_RadioTower_RedLights_Billboard
graph {
 properties {
  bUsedWithInstancedStaticMeshes=true
  bUsedWithNanite=true
  blendMode=BLEND_Translucent
  pixelDepthOffsetMode=PDOM_Legacy
  shadingModel=MSM_Unlit
  twoSided=true
  usedShadingModels=Unlit
 }
 nodes [
  n1 = AppendVector(A=Add(A=TextureCoordinate()) {constB=-0.5}, B=Constant())
  n2 = MaterialFunctionCall() {materialFunction=&ObjectScale}
  n3 = Constant() {r=-160}
  n4 = Transform(None=Multiply(A=NamedRerouteDeclaration(None=@n1) {name=TexcoorOrigin, nodeColor=<rgba 1 0 0.8203 1>}, B=Multiply(A=@n2.Scale X, B=@n3)) {constB=-160})
  n5 = Add(A=@n4, B=WorldPosition().XYZ)
  n6 = Distance(A=@n5, B=CameraPositionWS())
  n7 = Clamp(None=Multiply(A=@n6) {constB=0.0002}) {maxDefault=13, minDefault=4}
  n8 = NamedRerouteUsage()
  n9 = MaterialFunctionCall() {materialFunction=&/MF_RadioTower_Pulse}
  n10 = LinearInterpolate(
   A=Add(A=Multiply(A=@n9.Result)) {constB=0.3}
   Alpha=ScalarParameter() {bUseCustomPrimitiveData=true, parameterName=Activation Complete, primitiveDataIndex=2}
  )
  n11 = Max(A=Multiply(A=LinearInterpolate(Alpha=@n8) {constA=200, constB=300}, B=@n10), B=Multiply(A=NamedRerouteUsage()) {constB=200})
  n12 = AppendVector(A=@n11, B=Multiply(A=@n11) {constB=0.54})
  n13 = Constant()
  n14 = Multiply(A=Multiply(A=NamedRerouteUsage(), B=Constant3Vector() {constant=<rgba 1 -1 1 0>}), B=Multiply(A=Add(A=@n7), B=AppendVector(A=@n12, B=@n13)))
  n15 = ScalarParameter() {bUseCustomPrimitiveData=true, defaultValue=1, parameterName=Active State}
  n16 = Multiply(A=MaterialFunctionCall() {materialFunction=&ObjectScale}.Scale X, B=@n15)
  n17 = Add(A=@n4, B=Multiply(A=Transform(None=@n14) {transformSourceType=TRANSFORMSOURCE_View}, B=LinearInterpolate(A=@n16, B=@n15)))
  n18 = Subtract(A=Add(A=@n17, B=WorldPosition().XYZ), B=CameraPositionWS())
  n19 = MaterialFunctionCall() {materialFunction=&ObjectScale}
  n20 = Constant() {r=-100}
  Add(A=@n17, B=Multiply(A=Normalize(VectorInput=@n18), B=Multiply(A=@n19.Scale X, B=@n20)) {constB=-100})
  Constant() {r=0.1}
  n21 = Add(A=TextureCoordinate()) {constB=-0.5}
  n22 = ComponentMask(None=@n21) {r=true}
  n23 = Multiply(A=@n22) {constB=2}
  n24 = ComponentMask(None=@n21) {g=true}
  n25 = Add(A=Add(A=Multiply(A=@n23, B=@n23), B=Multiply(A=@n24, B=@n24))) {constB=0.499}
  n26 = OneMinus(None=Round(None=Saturate(None=@n25)))
  n27 = Saturate(None=Subtract(A=Multiply(A=@n6) {constB=0.0001}) {constB=4})
  n28 = VertexInterpolator(VS=Multiply(A=@n27, B=@n27))
  n29 = Multiply(A=Constant3Vector() {constant=<rgba 0.0716 0.0867 0.032 0>}, B=Abs(None=Round(None=@n5)))
  n30 = ComponentMask(None=@n29) {r=true}
  n31 = ComponentMask(None=@n29) {g=true}
  n32 = Frac(None=Add(A=Add(A=@n30, B=@n31), B=ComponentMask(None=@n29) {b=true}))
  n33 = Add(A=Multiply(A=Round(None=@n32)) {constB=2}) {constB=-1}
  n34 = Multiply(A=Time() {bOverride_Period=true, period=6.6667}) {constB=0.15}
  n35 = Multiply(A=TextureSample(UVs=Add(A=@n34, B=@n32)) {samplerType=SAMPLERTYPE_LinearGrayscale, texture=/MidMatchRadioTowerGameplay/Art/Textures/T_Gameplay_RadioTower_RedLights_Noise}.R) {constB=0.5}
  n36 = Multiply(A=AppendVector(A=Multiply(A=@n22, B=@n33) {constB=-1}, B=@n24), B=Add(A=@n35)) {constB=2}
  n37 = VertexInterpolator(VS=AppendVector(A=Add(A=@n36) {constB=0.5}, B=OneMinus(None=@n35)))
  n38 = TextureSample(UVs=ComponentMask(None=@n37.PS) {g=true, r=true}) {
   samplerSource=SSM_Clamp_WorldGroupSettings
   samplerType=SAMPLERTYPE_LinearGrayscale
   texture=/MidMatchRadioTowerGameplay/Art/Textures/T_Gameplay_RadioTower_RedLights_Opacity
  }
  n39 = Add(A=Multiply(A=Multiply(A=@n26) {constB=5}, B=ComponentMask(None=@n28.PS) {r=true}), B=@n38.R)
  n40 = Add(A=NamedRerouteUsage(), B=Multiply(A=@n39, B=ComponentMask(None=@n37.PS) {b=true}))
  n41 = Constant() {r=0.04}
  n42 = MaterialFunctionCall(Value_Morning=@n41, Value_Day=Constant() {r=0.084}, Value_Evening=@n41, Value_Night=@n41) {materialFunction=&Scalar_Time_of_Day}
  n43 = Max(A=NamedRerouteUsage(), B=NamedRerouteUsage())
  n44 = NamedRerouteUsage()
  DepthFade(Opacity=Multiply(A=Multiply(A=@n40, B=@n42.Result) {constB=0.25}, B=Max(A=@n43, B=@n44)))
  n45 = VectorParameter() {parameterName=Color1, defaultValue=<rgba 0.7083 0.2761 0.0232 1>}
  n46 = VectorParameter() {parameterName=Color2, defaultValue=<rgba 1 0.6295 0.2344 1>}
  n47 = NamedRerouteUsage()
  n48 = VectorParameter() {parameterName=Color5, defaultValue=<rgba 0 0.4195 1.5 1>}
  n49 = VectorParameter() {parameterName=Color6, defaultValue=<rgba 0 0.8 2 1>}
  n50 = NamedRerouteUsage()
  n51 = LinearInterpolate(
   A=LinearInterpolate(A=LinearInterpolate(A=@n45.RGB, B=@n46.RGB, Alpha=@n47), B=LinearInterpolate(A=@n48.RGB, B=@n49.RGB, Alpha=@n50), Alpha=NamedRerouteUsage())
   B=LinearInterpolate(
    A=VectorParameter() {parameterName=Color3, defaultValue=<rgba 0.2188 1.2443 3 1>}.RGB
    B=VectorParameter() {parameterName=Color4, defaultValue=<rgba 1.2 1.2 1.2 1>}.RGB
    Alpha=NamedRerouteUsage()
   )
   Alpha=NamedRerouteUsage()
  )
  n52 = LinearInterpolate(
   A=LinearInterpolate(
    A=VectorParameter() {parameterName=Color7, defaultValue=<rgba 0.0077 0.4375 0.0378 1>}.RGB
    B=VectorParameter() {parameterName=Color8, defaultValue=<rgba 0.0607 1.4 0.1778 1>}.RGB
    Alpha=NamedRerouteUsage()
   )
   B=LinearInterpolate(
    A=VectorParameter() {parameterName=Color9, defaultValue=<rgba 0.5208 0 0.0228 1>}.RGB
    B=VectorParameter() {parameterName=Color10, defaultValue=<rgba 1.2 0.075 0.0857 1>}.RGB
    Alpha=NamedRerouteUsage()
   )
   Alpha=ScalarParameter() {bUseCustomPrimitiveData=true, parameterName=Red Lights, primitiveDataIndex=3}
  )
  n53 = ScalarParameter() {bUseCustomPrimitiveData=true, parameterName=Activation Complete, primitiveDataIndex=2}
  Multiply(A=Multiply(A=LinearInterpolate(A=@n51, B=@n52, Alpha=@n53), B=NamedRerouteUsage())) {constB=1.2}
  n54 = TextureCoordinate()
  n55 = ComponentMask(None=@n54) {g=true}
  Multiply(A=@n55)
  n56 = VertexColor()
  NamedRerouteDeclaration(None=Floor(None=Add(A=@n56.R) {constB=0.01})) {name=TopLight, nodeColor=<rgba 0 1 0.7422 1>}
  n57 = Add(A=OneMinus(None=@n56.R)) {constB=-0.01}
  n58 = ScalarParameter() {bUseCustomPrimitiveData=true, defaultValue=1, parameterName=ActivationLights, primitiveDataIndex=1}
  NamedRerouteDeclaration(None=Floor(None=Add(A=@n57, B=@n58) {constB=0.5})) {name=Activation Lights, nodeColor=<rgba 0.6406 0 1 1>}
  NamedRerouteDeclaration(None=VertexInterpolator(VS=@n32).PS) {name=Random Static, nodeColor=<rgba 1 0.4922 0 1>}
  NamedRerouteDeclaration(None=ComponentMask(None=VertexInterpolator(VS=@n10).PS) {r=true}) {name=Scroll Pulse, nodeColor=<rgba 1 0.6164 0 1>}
  n59 = TextureCoordinate()
  n60 = Add(A=Time(), B=NamedRerouteUsage())
  n61 = Multiply(
   A=TextureSample(UVs=Rotator(Coordinate=@n59, Time=@n60) {speed=-0.7}) {samplerType=SAMPLERTYPE_LinearColor, texture=/MidMatchRadioTowerGameplay/Art/Textures/T_Gameplay_RadioTower_RedLights_Chroma}.RGB
   B=TextureSample(UVs=Rotator(Coordinate=@n59, Time=@n60) {speed=1}) {samplerType=SAMPLERTYPE_LinearColor, texture=&T_Gameplay_RadioTower_RedLights_Chroma}.RGB
  ) {constB=2}
  NamedRerouteDeclaration(None=Multiply(A=LinearInterpolate(A=@n61) {constAlpha=0.4}) {constB=2}) {name=Chroma, nodeColor=<rgba 1 0 0.375 1>}
  NamedRerouteDeclaration(None=Multiply(A=@n39, B=@n39)) {name=Flare Mask, nodeColor=<rgba 0.5937 1 0 1>}
  n62 = ComponentMask(None=@n54) {r=true}
  n63 = Add(A=@n62) {constB=-0.5}
  n64 = OneMinus(None=Multiply(A=Abs(None=@n63)) {constB=2})
  n65 = Multiply(A=Time()) {constB=0.1}
  n66 = Subtract(A=Multiply(A=@n62) {constB=3}, B=@n65)
  n67 = NamedRerouteUsage()
  n68 = TextureSample(UVs=AppendVector(A=Add(A=@n66, B=@n67), B=@n55)) {samplerType=SAMPLERTYPE_LinearColor, texture=/MidMatchRadioTowerGameplay/Art/Textures/T_Gameplay_RadioTower_RedLights_Dots}
  n69 = Multiply(A=@n62) {constB=1.5}
  n70 = Add(A=@n69, B=@n65)
  n71 = NamedRerouteUsage()
  n72 = TextureSample(UVs=AppendVector(A=Add(A=@n70, B=@n71), B=@n55)) {samplerType=SAMPLERTYPE_LinearColor, texture=&T_Gameplay_RadioTower_RedLights_Dots}
  n73 = Multiply(A=Multiply(A=@n64, B=@n64), B=Multiply(A=Multiply(A=@n68.R, B=@n72.R)) {constB=1.2})
  NamedRerouteDeclaration(None=Multiply(A=@n73, B=@n73)) {name=Sparks, nodeColor=<rgba 0 1 0.1797 1>}
  Subtract(A=@n69, B=Multiply(A=Time()) {constB=0.1})
 ]
}
