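# Opaque, Nanite-ready lit material for radio tower lights, showing active, activation-progress, completed and red-light states through color and glow.
# Custom primitive data drives most state blends; vertex-color red yields top-light and stepped activation masks.
# Emissive combines state colors and a pulse function, shaped by squared horizontal view-facing falloff and scaled by time of day.
# Base color's Activation Complete parameter, unlike the emissive's, lacks custom primitive data.
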Material /MidMatchRadioTowerGameplay/Art/Materials/M_RadioTower_RedLights_Opaque
graph {
 properties {bUsedWithInstancedStaticMeshes=true, bUsedWithNanite=true, pixelDepthOffsetMode=PDOM_Legacy, usedShadingModels=Default Lit}
 nodes [
  Constant() {r=0.3}
  n1 = LinearInterpolate(
   A=VectorParameter() {parameterName=Color3, defaultValue=<rgba 0.0781 0.0519 0.0252 1>}.RGB
   B=VectorParameter() {parameterName=Color6, defaultValue=<rgba 0 0.0632 0.0781 1>}.RGB
   Alpha=NamedRerouteUsage()
  )
  n2 = VectorParameter() {parameterName=Color4, defaultValue=<rgba 0.3802 0.3802 0.3802 1>}
  n3 = NamedRerouteUsage()
  n4 = VectorParameter() {parameterName=Color7, defaultValue=<rgba 0.0347 0.1667 0.1141 1>}
  n5 = VectorParameter() {parameterName=Color8, defaultValue=<rgba 0.125 0.0187 0.015 1>}
  n6 = ScalarParameter() {bUseCustomPrimitiveData=true, parameterName=Red Lights, primitiveDataIndex=3}
  LinearInterpolate(
   A=Constant3Vector() {constant=<rgba 0.1682 0.1719 0.1441 1>}
   B=LinearInterpolate(
    A=LinearInterpolate(A=@n1, B=@n2.RGB, Alpha=@n3)
    B=LinearInterpolate(A=@n4.RGB, B=@n5.RGB, Alpha=@n6)
    Alpha=ScalarParameter() {parameterName=Activation Complete}
   )
   Alpha=ScalarParameter() {bUseCustomPrimitiveData=true, parameterName=Active State}
  )
  n7 = ComponentMask(None=PixelNormalWS()) {g=true, r=true}
  n8 = ComponentMask(None=CameraVectorWS()) {g=true, r=true}
  n9 = Saturate(None=DotProduct(A=Normalize(VectorInput=@n7), B=Normalize(VectorInput=@n8)))
  n10 = LinearInterpolate(
   A=LinearInterpolate(
    A=VectorParameter() {parameterName=Color1, defaultValue=<rgba 1 0.5509 0.0261 1>}.RGB
    B=VectorParameter() {parameterName=Color5, defaultValue=<rgba 0 0.6815 1 1>}.RGB
    Alpha=NamedRerouteUsage()
   )
   B=VectorParameter() {parameterName=Color2, defaultValue=<rgba 0.9685 0.9571 1 1>}.RGB
   Alpha=NamedRerouteUsage()
  )
  n11 = Max(
   A=Add(A=MaterialFunctionCall() {materialFunction=&/MF_RadioTower_Pulse}.Result) {constB=0.02}
   B=NamedRerouteUsage()
  )
  n12 = VectorParameter() {parameterName=Color9, defaultValue=<rgba 0 1 0.1721 1>}
  n13 = VectorParameter() {parameterName=Color10, defaultValue=<rgba 1 0.0098 0 1>}
  n14 = ScalarParameter() {bUseCustomPrimitiveData=true, parameterName=Red Lights, primitiveDataIndex=3}
  n15 = Multiply(
   A=Multiply(A=@n9, B=@n9)
   B=LinearInterpolate(
    A=Multiply(A=@n10, B=@n11)
    B=LinearInterpolate(A=@n12.RGB, B=@n13.RGB, Alpha=@n14)
    Alpha=ScalarParameter() {bUseCustomPrimitiveData=true, parameterName=Activation Complete, primitiveDataIndex=2}
   ) {constAlpha=0}
  )
  n16 = Constant() {r=15}
  n17 = MaterialFunctionCall(Value_Morning=@n16, Value_Day=@n16, Value_Evening=@n16, Value_Night=Constant() {r=4}) {materialFunction=&Scalar_Time_of_Day}
  Multiply(
   A=Subtract(A=Multiply(A=@n15, B=@n17.Result), B=Multiply(A=@n17.Result) {constB=0.05})
   B=ScalarParameter() {bUseCustomPrimitiveData=true, parameterName=Active State}
  )
  n18 = VertexColor()
  NamedRerouteDeclaration(None=Floor(None=Add(A=@n18.R) {constB=0.01})) {name=TopLight, nodeColor=<rgba 0 1 0.7422 1>}
  n19 = Add(A=OneMinus(None=@n18.R)) {constB=-0.01}
  n20 = ScalarParameter() {bUseCustomPrimitiveData=true, parameterName=ActivationLights, primitiveDataIndex=1}
  NamedRerouteDeclaration(None=Floor(None=Add(A=@n19, B=@n20) {constB=0.5})) {name=Activation Lights, nodeColor=<rgba 1 0.042 0.7886 1>}
  TextureSample() {samplerType=SAMPLERTYPE_Normal, texture=/MidMatchRadioTowerGameplay/Art/Textures/T_Gameplay_RadioTower_RedLights_N}
  VertexNormalWS()
 ]
}
